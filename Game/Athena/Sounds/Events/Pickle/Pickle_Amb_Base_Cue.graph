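# Looping ambience cue for the Pickle event, reacting to OpenAmount and Destabilize parameters.
# OpenAmount shapes pitch and volume across layered open and closed base loops and two destabilize loops, while Destabilize crossfades from the base mix through both destabilize loops.
# A parameter-supplied wave is mixed in, scaled by OpenAmount.
# That wave passes through a ShouldMuteLicensedMusic branch and licensed-content switcher, which can substitute silence.
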SoundCue /Game/Athena/Sounds/Events/Pickle/Pickle_Amb_Base_Cue
root=ModulatorContinuous_4
nodes [
 {
  id=Branch_0
  children [FortSoundNodeLicensedContentSwitcher_0, WaveParam_0, WaveParam_0]
  properties {boolParameterName=ShouldMuteLicensedMusic}
 }
 {id=FortSoundNodeLicensedContentSwitcher_0, class=FortSoundNodeLicensedContentSwitcher, children [WavePlayer_2,WavePlayer_2,WaveParam_0]}
 {id=Mixer_0, children [ModulatorContinuous_3, ParamCrossFade_4], properties {inputVolume [1,1]}}
 {id=Mixer_2, children [ModulatorContinuous_2, ModulatorContinuous_0], properties {inputVolume [1,1]}}
 {
  id=ModulatorContinuous_0
  children [WavePlayer_0]
  properties {pitchModulationParams {default=0, maxOutput=0.5, minOutput=1, parameterName=OpenAmount}}
 }
 {
  id=ModulatorContinuous_1
  children [WavePlayer_1]
  properties {
   pitchModulationParams {default=0, maxInput=0.1, minOutput=0.7, parameterName=OpenAmount}
   volumeModulationParams {default=0, maxInput=0.1, minOutput=0.2, parameterName=OpenAmount}
  }
 }
 {
  id=ModulatorContinuous_2
  children [ModulatorContinuous_1]
  properties {
   pitchModulationParams {default=0, maxOutput=0.8, minInput=0.5, minOutput=1, parameterName=OpenAmount}
   volumeModulationParams {default=0, maxOutput=0.8, minInput=0.5, minOutput=1, parameterName=OpenAmount}
  }
 }
 {
  id=ModulatorContinuous_3
  children [Branch_0]
  properties {
   pitchModulationParams {default=0, maxInput=0.3, minOutput=1, parameterName=OpenAmount}
   volumeModulationParams {default=0, maxInput=0.2, parameterName=OpenAmount}
  }
 }
 {
  id=ModulatorContinuous_4
  children [Mixer_0]
  properties {
   pitchModulationParams {default=0, minOutput=1, parameterName=Destabilize}
   volumeModulationParams {default=0, minOutput=1, parameterName=Destabilize}
  }
 }
 {
  id=ModulatorContinuous_5
  children [WavePlayer_5]
  properties {
   pitchModulationParams {default=0, maxOutput=1.25, minOutput=1, parameterName=OpenAmount}
   volumeModulationParams {default=0, minOutput=0.5, parameterName=OpenAmount}
  }
 }
 {
  id=ModulatorContinuous_6
  children [WavePlayer_6]
  properties {
   pitchModulationParams {default=0, maxOutput=1.25, minOutput=1, parameterName=OpenAmount}
   volumeModulationParams {default=0, minOutput=0.5, parameterName=OpenAmount}
  }
 }
 {
  id=ParamCrossFade_4
  children [Mixer_2, ModulatorContinuous_5, ModulatorContinuous_6]
  properties {
   paramName=Destabilize
   crossFadeInput [
    {fadeInDistanceEnd=0, fadeInDistanceStart=0, fadeOutDistanceEnd=0.5, fadeOutDistanceStart=0, volume=1}
    {fadeInDistanceEnd=0.5, fadeInDistanceStart=0, fadeOutDistanceEnd=1, fadeOutDistanceStart=0.5, volume=1}
    {fadeInDistanceEnd=1, fadeInDistanceStart=0.5, fadeOutDistanceEnd=100, fadeOutDistanceStart=1, volume=1}
   ]
  }
 }
 {id=WaveParam_0, children [null], properties {waveParameterName=Asset}}
 {id=WavePlayer_0, properties {bLooping=true, soundWaveAssetPtr=&/Pickle_Amb_Base_Closed_Loop_01}}
 {id=WavePlayer_1, properties {bLooping=true, soundWaveAssetPtr=&/Pickle_Amb_Open_Base_Loop_01}}
 {id=WavePlayer_2, properties {bLooping=true, soundWaveAssetPtr=/Game/Sounds/Fort_Test_Sounds/5_seconds_silence}}
 {id=WavePlayer_5, properties {bLooping=true, soundWaveAssetPtr=&/Pickle_Amb_Destabilize_01_Loop}}
 {id=WavePlayer_6, properties {bLooping=true, soundWaveAssetPtr=&/Pickle_Amb_Destabilize_02_Loop}}
]
properties {
 attenuationSettings=&/Pickle_Amb_Base_Att
 maxDistance=10512
 soundClassObject=&Fort_Events
 sourceEffectChain=&/Pickle_SrcFX_Chain
 volumeMultiplier=0.6
}
